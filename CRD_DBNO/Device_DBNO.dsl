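; Creative Down But Not Out device that configures knockdown, carry and revive rules for a minigame.
; While enabled, it pushes its knobs into the minigame's DBNO settings mutator on the server.
; It applies a gameplay effect to participants and binds their pawn DBNO delegates to broadcast downed, revived, carry and shakedown events and drop items.
; Receivers toggle the device, down players via a kill effect, or force revives, making any hoister drop them.
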(event Custom|EnabledReceivedFunction
  (params (Instigator :type Controller%20Object%20Reference))
  ;[0 0.8871 1] Enable / Disable receivers
  (Networking|SwitchHasAuthority
    (:Authority
      (Class|CreativeEnabledComponent|SetEnabled (Variables|Default|GetEnabledComponent) true))))

(event Custom|DisableReceivedFunction
  (params (Instigator :type Controller%20Object%20Reference))
  ;[0 0.8871 1] Enable / Disable receivers
  (Networking|SwitchHasAuthority
    (:Authority
      (Class|CreativeEnabledComponent|SetEnabled (Variables|Default|GetEnabledComponent) false))))

(event OnEnabledStateChanged(EnabledComponent)
  (params (Enabled :type Boolean))
  (bind _fortminigameprogress (Variables|Default|GetFortMinigameProgress))
  (bind _currentminigame (Variables|/Script/FortniteGame.FortMinigameLogicComponent|GetCurrentMinigame _fortminigameprogress))
  ;[0 1 0.05153] Enable state changed
  (Networking|SwitchHasAuthority
    (:Authority
      (Utilities|IsValid _currentminigame
        (:Is%20Valid
          (if Enabled
            (CallFunction|SetDBNOSettings)
            (CallFunction|ApplyDBNOCreative)
            (else
              (CallFunction|DisableDevice _currentminigame))))))))

(event OnAddedtoMinigame(FortMinigameProgress)
  (params (Minigame :type Fort%20Minigame%20Object%20Reference))
  ;[1 1 0] Added / Removed from minigame
  (Networking|SwitchHasAuthority
    (:Authority
      (Minigame|AssignOnMinigamePlayerSpawned Minigame (EventDispatchers|CreateEvent :!event OnPlayerSpawned))
      (Minigame|AssignPlayerAddedPostResetDelegate Minigame (EventDispatchers|CreateEvent :!event OnPlayerAddedToMinigame))
      (Minigame|BindEventtoPlayerRemovedDelegate Minigame (EventDispatchers|CreateEvent :!event Player%20Removed%20from%20Minigame))
      (CallFunction|ApplyDBNOCreative)
      (if (Class|CreativeEnabledComponent|IsEnabled (Variables|Default|GetEnabledComponent))
        (CallFunction|SetDBNOSettings)))))

(event OnRemovedfromMinigame(FortMinigameProgress)
  (params (Minigame :type Fort%20Minigame%20Object%20Reference))
  ;[1 1 0] Added / Removed from minigame
  (Networking|SwitchHasAuthority
    (:Authority
      (Minigame|UnbindEventfromOnMinigamePlayerSpawned Minigame (EventDispatchers|CreateEvent :!event OnPlayerSpawned))
      (Minigame|UnbindEventfromPlayerAddedPostResetDelegate Minigame (EventDispatchers|CreateEvent :!event OnPlayerAddedToMinigame))
      (Minigame|UnbindEventfromPlayerRemovedDelegate Minigame (EventDispatchers|CreateEvent :!event Player%20Removed%20from%20Minigame))
      (CallFunction|DisableDevice Minigame))))

(event Custom|DownPlayerReceivedFunction
  (params (Instigator :type Controller%20Object%20Reference))
  ;[1 0.274 0] Receivers
  (Networking|SwitchHasAuthority
    (:Authority
      (if (Class|CreativeEnabledComponent|IsEnabled (Variables|Default|GetEnabledComponent))
        (bind _asplayer_controller (Utilities|Casting|CastToPlayerController Instigator))
        (if (Default|IsDBNOEnabled (Variables|/Script/Engine.Controller|GetPlayerState _asplayer_controller))
          (bind _asfort_player_pawn_athena (Utilities|Casting|CastToFortPlayerPawnAthena (Pawn|GetControlledPawn Instigator)))
          (if (not (Health|IsDBNO _asfort_player_pawn_athena))
            (GameplayEffects|ApplyGameplayEffectToSelf :self (Ability|GetAbilitySystemComponent Instigator) :GameplayEffectClass (Variables|GameplayEffects|GetGE_KillPlayer) :Level 0.0)))))))

(event Custom|RevivePlayerReceivedFunction
  (params (Instigator :type Controller%20Object%20Reference))
  ;[1 0.274 0] Receivers
  (Networking|SwitchHasAuthority
    (:Authority
      (if (Class|CreativeEnabledComponent|IsEnabled (Variables|Default|GetEnabledComponent))
        (bind _asfort_player_pawn_athena (Utilities|Casting|CastToFortPlayerPawnAthena (Pawn|GetControlledPawn Instigator)))
        (if (Health|IsDBNO _asfort_player_pawn_athena)
          (Class|FortPlayerPawnAthena|ForceRevivefromDBNO _asfort_player_pawn_athena Instigator)
          (Transmitters|PawnRevivedfromDBNO _asfort_player_pawn_athena)
          (Utilities|IsValid (Ability|GetAbilitySystemComponent (Class|FortPlayerPawn|GetDBNOHoister _asfort_player_pawn_athena))
            (:Is%20Valid
              (Abilities|TryActivateAbilitybyClass (Ability|GetAbilitySystemComponent (Class|FortPlayerPawn|GetDBNOHoister _asfort_player_pawn_athena)) "/Game/Abilities/NPC/Generic/GAB_DropPlayer.GAB_DropPlayer_C" true))))))))

(fn ApplyCreativeDBNOtoPlayer
  (params (Pawn :type Pawn%20Object%20Reference))
  (GameplayEffects|ApplyGameplayEffectToSelf :self (Ability|GetAbilitySystemComponent Pawn) :GameplayEffectClass (Variables|GameplayEffects|GetGE_ApplyCreativeDBNO) :Level 0.0))

(fn ApplyDBNOCreative
  (for _array_element (Minigame|GetParticipatingPlayers (Variables|/Script/FortniteGame.FortMinigameLogicComponent|GetCurrentMinigame (Variables|Default|GetFortMinigameProgress)))
    (Transmitters|BindtoPawnDBNOStatesChange (Variables|/Script/Engine.PlayerState|GetPawnPrivate _array_element))
    (CallFunction|ApplyCreativeDBNOToPlayer (Variables|/Script/Engine.PlayerState|GetPawnPrivate _array_element))))

(fn BindToPawnDBNOStatesChange
  (params (Pawn :type Pawn%20Object%20Reference))
  (bind _asplayer_pawn_athena (Utilities|Casting|CastToPlayerPawn_Athena Pawn)
    (:then
      (EventDispatchers|BindEventtoOnDBNOStateChanged _asplayer_pawn_athena (EventDispatchers|CreateEvent :!event PawnDBNOStateChanged))
      (EventDispatchers|BindEventtoOnRevivedfromDBNODelegate _asplayer_pawn_athena (EventDispatchers|CreateEvent :!event PawnRevivedFromDBNO))
      (EventDispatchers|BindEventtoOnDBNOHoisterChangedDelegate _asplayer_pawn_athena (EventDispatchers|CreateEvent :!event PawnDBNOHoisterChanged)))
    (:CastFailed
      _))
  (return))

(fn DisableDevice
  (params (Minigame :type Fort%20Minigame%20Object%20Reference))
  (Class|FortAthenaMutatorDBNOSetting|UnsetDBNOSettings (Class|FortAthenaMutatorDBNOSetting|GetDBNOSettingsMutator Minigame) self)
  (Default|RemoveDBNOCreative Minigame))

(fn DropItemsFromPlayer
  (params (DropItemSetting :type Enum_DBNO_DropItemSetting%20Enum :default "NewEnumerator0") (PlayerPawn :type Fort%20Player%20Pawn%20Object%20Reference))
  (bind _emptycreativetags (LocalVariables|GetEmptyCreativeTags))
  (Utilities|IsValid PlayerPawn
    (:Is%20Valid
      (bind _asfort_player_controller (Utilities|Casting|CastToFortPlayerController (Pawn|GetController PlayerPawn)))
      (switch Utilities|FlowControl|Switch|SwitchonEnum_DBNO_DropItemSetting (Variables|GetDropItemSetting)
        (:NewEnumerator1
          (Inventory|DropAllItems _asfort_player_controller 0 0 false true :!class /Script/FortniteGame.FortPlayerController))
        (:NewEnumerator2
          (Inventory|DropItemsbyCreativeTags _asfort_player_controller (Variables|Default|GetBuildingMaterialItemDropTags) _emptycreativetags))
        (:NewEnumerator3
          (Inventory|DropItemsbyCreativeTags _asfort_player_controller (Variables|Default|GetGameResourceItemDropTags) _emptycreativetags))))))

(fn GetScaleMax
  (params (DefaultScaleMax :type Float%20%28single-precision%29))
  (return 1.0))

(fn GetScaleMin
  (params (DefaultScaleMin :type Float%20%28single-precision%29))
  (return 1.0))

(fn GetVisibilityComponentsWithCollisionOverrides
  (return (Utilities|Map|MakeMap (Variables|Debug|GetStaticMeshComponent) "NoCollision" 0 "NoCollision")))

(fn InitializeDevice
  (bind _fortminigameprogress (Variables|Default|GetFortMinigameProgress))
  (bind _currentminigame (Variables|/Script/FortniteGame.FortMinigameLogicComponent|GetCurrentMinigame _fortminigameprogress))
  (|CallParentFunction)
  (Networking|SwitchHasAuthority
    (:Authority
      (Utilities|IsValid _currentminigame
        (:Is%20Valid
          (Class|CreativeEnabledComponent|SetEnabledOnPhase (Variables|Default|GetEnabledComponent) (Variables|/Script/FortniteGame.FortMinigame|GetCurrentState _currentminigame) (Variables|LegacyOptions|GetEnabledIndex))
          (if (Class|CreativeEnabledComponent|IsEnabled (Variables|Default|GetEnabledComponent))
            (CallFunction|SetDBNOSettings)
            (CallFunction|ApplyDBNOCreative)))))))

(fn IsDBNOEnabled
  (params (PlayerState :type Player%20State%20Object%20Reference))
  (bind _asfort_game_state_athena (Utilities|Casting|CastToFortGameStateAthena (Game|GetGameState))
    (:then
      (bind _asfort_player_state (Utilities|Casting|CastToFortPlayerState PlayerState)
        (:then
          (return (and (Utilities|Enum|NotEqual(Enum) (Utilities|Struct|BreakDBNOCustomSettings (Class|FortGameStateAthena|EvaluateDBNOMutator _asfort_game_state_athena _asfort_player_state)) "Off") (.ReturnValue (Class|FortGameStateAthena|EvaluateDBNOMutator _asfort_game_state_athena _asfort_player_state)))))
        (:CastFailed
          _)))
    (:CastFailed
      _))
  (return false))

(fn OnPlayerAddedToMinigame
  (params (PlayerNetid :type Unique%20Net%20Id%20Structure) (IsLocalPlayer :type Boolean))
  (bind _fort_player_state (Class|CreativeCommonDeviceFunctionLibrary|GetPlayerStateFromNetID PlayerNetid))
  (bind _pawnprivate (Variables|/Script/Engine.PlayerState|GetPawnPrivate _fort_player_state))
  (Transmitters|BindtoPawnDBNOStatesChange _pawnprivate)
  (CallFunction|ApplyCreativeDBNOToPlayer _pawnprivate))

(fn OnPlayerSpawned
  (params (PlayerController :type Controller%20Object%20Reference))
  (bind _returnvalue (Pawn|GetControlledPawn PlayerController))
  (Transmitters|BindtoPawnDBNOStatesChange _returnvalue)
  (CallFunction|ApplyCreativeDBNOToPlayer _returnvalue))

(fn PawnDBNOHoisterChanged
  (params (CarryEvent :type EFortDBNOCarryEvent%20Enum :default "PickedUp") (HoisterPawn :type Fort%20Player%20Pawn%20Object%20Reference) (HoisteePawn :type Fort%20Player%20Pawn%20Object%20Reference))
  (bind (_value _returnvalue) (Utilities|Map|Find (Variables|Default|GetPlayersLastState) (Variables|/Script/Engine.Pawn|GetPlayerState (Variables|GetHoisteePawn))))
  (Utilities|FlowControl|Sequence
    (:then_0
      (switch Utilities|FlowControl|Switch|SwitchonEFortDBNOCarryEvent CarryEvent
        (:PickedUp
          (GameplayEvent|BroadcastEvent (Variables|UserOptions|Events|GetOn%20Player%20Picked%20Up) 0 (Utilities|Struct|MakeGameplayEventPlayerPickedUp (Pawn|GetController (Variables|GetHoisteePawn)))))
        (:Interrogating
          (GameplayEvent|BroadcastEvent (Variables|UserOptions|Events|GetOn%20Shaken%20Down) 0 (Utilities|Struct|MakeGameplayEventShakenDown (Pawn|GetController (Variables|GetHoisteePawn))))
          (GameplayEvent|BroadcastEvent (Variables|UserOptions|Events|GetOn%20Shakedown) 0 (Utilities|Struct|MakeGameplayEventShakedown (Pawn|GetController (Variables|GetHoisterPawn))))
          (CallFunction|DropItemsfromPlayer (Variables|Knobs|GetDropItemsOnShakenDown) (Variables|GetHoisterPawn)))
        (:Dropped
          (if (and (Utilities|Enum|Equal(Enum) _value "PickedUp") _returnvalue)
            (GameplayEvent|BroadcastEvent (Variables|UserOptions|Events|GetOn%20Player%20Dropped) 0 (Utilities|Struct|MakeGameplayEventPlayerDropped (Pawn|GetController (Variables|GetHoisteePawn))))))
        (:Thrown
          (GameplayEvent|BroadcastEvent (Variables|UserOptions|Events|GetOn%20Player%20Thrown) 0 (Utilities|Struct|MakeGameplayEventPlayerThrown (Pawn|GetController (Variables|GetHoisteePawn)))))))
    (:then_1
      (Utilities|Map|Add (Variables|Default|GetPlayersLastState) (Variables|/Script/Engine.Pawn|GetPlayerState (Variables|GetHoisteePawn)) (Variables|GetCarryEvent)))))

(fn PawnDBNOStateChanged
  (params (FortPawn :type Fort%20Pawn%20Object%20Reference) (bInIsDBNO :type Boolean))
  (if bInIsDBNO
    (if (Knobs|Team|IsTeamNone (Variables|Knobs|GetSelectedTeam))
      _
      (elif (xor (== (Variables|Knobs|GetKnob_AffectsTeam) 0) (== (Team|GetActorPvPTeamIndex (Variables|GetFortPawn)) (Knobs|Team|GetTeamValueasIndex (Variables|Knobs|GetSelectedTeam))))
        _))
    (GameplayEvent|BroadcastEvent (Variables|UserOptions|Events|GetOn%20Player%20Downed) 0 (Utilities|Struct|MakeGameplayEventPlayerDowned (Pawn|GetController (Variables|GetFortPawn))))
    (bind _asfort_player_pawn (Utilities|Casting|CastToFortPlayerPawn (Variables|GetFortPawn)))
    (CallFunction|DropItemsfromPlayer (Variables|Knobs|GetDropItemsOnKnockDown) _asfort_player_pawn)))

(fn PawnRevivedFromDBNO
  (params (RevivedPawn :type Fort%20Player%20Pawn%20Object%20Reference))
  (GameplayEvent|BroadcastEvent (Variables|UserOptions|Events|GetOn%20Player%20Revived) 0 (Utilities|Struct|MakeGameplayEventPlayerRevived (Pawn|GetController RevivedPawn))))

(fn PlayerRemovedfromMinigame
  (params (UniqueNetId :type Unique%20Net%20Id%20Structure) (bIsLocalPlayer :type Boolean))
  (Transmitters|UnBindtoPawnDBNOStatesChange (Variables|/Script/Engine.PlayerState|GetPawnPrivate (Class|CreativeCommonDeviceFunctionLibrary|GetPlayerStateFromNetID UniqueNetId))))

(fn RemoveDBNOCreative
  (params (Minigame :type Fort%20Minigame%20Object%20Reference))
  (for _array_element (Minigame|GetParticipatingPlayers Minigame)
    (Transmitters|UnBindtoPawnDBNOStatesChange (Variables|/Script/Engine.PlayerState|GetPawnPrivate _array_element))))

(fn SetDBNOSettings
  (Class|FortAthenaMutatorDBNOSetting|SetDBNOSettingsandOverride (Class|FortAthenaMutatorDBNOSetting|GetDBNOSettingsMutator (Variables|/Script/FortniteGame.FortMinigameLogicComponent|GetCurrentMinigame (Variables|Default|GetFortMinigameProgress))) (Utilities|Struct|MakeDBNOCustomSettings (Variables|Knobs|GetDBNOEnabled) (Variables|Knobs|GetTenacityDepletionRate) (Variables|Knobs|GetReviveHealthPercentage) (Variables|Knobs|GetbAllowRevives) (Variables|Knobs|GetbAllowCarry) (Variables|Knobs|GetTimeToRevive) (Variables|Knobs|GetbAlertTeamWhenDowned) (Variables|Knobs|GetbLastManStandingMode) (Variables|Knobs|GetbAllowInterrogations) (Variables|Knobs|GetbShowPlayerLocationInterrogation) (Variables|Knobs|GetSelectedTeam) (== (Variables|Knobs|GetKnob_AffectsTeam) 0) (Variables|Knobs|GetKnob_SelectedClass) (== (Variables|Knobs|GetKnob_AffectsClass) 0) (Variables|Knobs|GetTenacityType) (Variables|Knobs|GetTenacityAmount)) self))

(fn UnBindToPawnDBNOStatesChange
  (params (Pawn :type Pawn%20Object%20Reference))
  (bind _asplayer_pawn_athena (Utilities|Casting|CastToPlayerPawn_Athena Pawn)
    (:then
      (EventDispatchers|UnbindEventfromOnDBNOStateChanged _asplayer_pawn_athena (EventDispatchers|CreateEvent :!event PawnDBNOStateChanged))
      (EventDispatchers|UnbindEventfromOnDBNOHoisterChangedDelegate _asplayer_pawn_athena (EventDispatchers|CreateEvent :!event PawnDBNOHoisterChanged))
      (EventDispatchers|UnbindEventfromOnRevivedfromDBNODelegate _asplayer_pawn_athena (EventDispatchers|CreateEvent :!event PawnRevivedFromDBNO)))
    (:CastFailed
      _))
  (return))

(fn UpdateVisibilityInGame)

(fn ConstructionScript
  (|CallParentFunction))
