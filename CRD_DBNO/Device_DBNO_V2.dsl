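; Creative Down But Not Out device (V2) that configures knock-down rules for minigame players by team and class, with receivers to down, revive, enable or disable.
; While enabled, it pushes user options into the minigame's DBNO mutator on the server and applies a gameplay effect to participants.
; Pawn delegates drive downed, revived, carry and shakedown events, plus optional item drops.
; Legacy option migration and version fixups preserve older placed devices' behavior.
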(event EventInitializeDevice
  (bind _fortminigameprogress (Variables|Default|GetFortMinigameProgress))
  (bind _currentminigame (Variables|/Script/FortniteGame.FortMinigameLogicComponent|GetCurrentMinigame _fortminigameprogress))
  ;[0.03244 1 0.2556 0.2] Initialization
  (|CallParentFunction)
  (Networking|SwitchHasAuthority
    (:Authority
      (Utilities|IsValid _currentminigame
        (:Is%20Valid
          (Class|CreativeEnabledComponent|SetEnabledOnPhase (Variables|Default|GetEnabledComponent) (Variables|/Script/FortniteGame.FortMinigame|GetCurrentState _currentminigame) (ChildFunctions|GetEnabledStateasIndex))
          (Initialization|Initialize (Variables|Default|GetCreative_UseRestriction_Component) (Variables|Default|GetFortMinigameProgress) (Variables|Default|GetEnabledComponent) 0 (Variables|UserOptions|GetSelectedTeam) (Variables|UserOptions|GetInvertTeamSelection) (Variables|UserOptions|GetSelected%20Class) (Variables|UserOptions|GetInvertClassSelection) false)
          (if (Class|CreativeEnabledComponent|IsEnabled (Variables|Default|GetEnabledComponent))
            (CallFunction|SetDBNOSettings)
            (CallFunction|ApplyDBNOCreative)))))))

(event OnEnabledStateChanged(EnabledComponent)
  (params (Enabled :type Boolean))
  (bind _fortminigameprogress (Variables|Default|GetFortMinigameProgress))
  (bind _currentminigame (Variables|/Script/FortniteGame.FortMinigameLogicComponent|GetCurrentMinigame _fortminigameprogress))
  ;[0.2086 0.3763 1 0.25] Enable state changed
  (Networking|SwitchHasAuthority
    (:Authority
      (Utilities|IsValid _currentminigame
        (:Is%20Valid
          (if Enabled
            (CallFunction|SetDBNOSettings)
            (CallFunction|ApplyDBNOCreative)
            (else
              (CallFunction|DisableDevice _currentminigame))))))))

(event OnAddedtoMinigame(FortMinigameProgress)
  (params (Minigame :type Fort%20Minigame%20Object%20Reference))
  ;[1 1 0] Added / Removed from minigame
  (Networking|SwitchHasAuthority
    (:Authority
      (Minigame|AssignOnMinigamePlayerSpawned Minigame (EventDispatchers|CreateEvent :!event OnParticipantSpawned))
      (Minigame|AssignPlayerAddedPostResetDelegate Minigame (EventDispatchers|CreateEvent :!event OnPlayerAddedToMinigame))
      (Minigame|BindEventtoPlayerRemovedDelegate Minigame (EventDispatchers|CreateEvent :!event Player%20Removed%20from%20Minigame))
      (Minigame|BindEventtoOnFinishedAddingMinigamePlayerDelegate Minigame (EventDispatchers|CreateEvent :!event OnFinishedAddingMinigamePlayer))
      (CallFunction|ApplyDBNOCreative)
      (if (Class|CreativeEnabledComponent|IsEnabled (Variables|Default|GetEnabledComponent))
        (CallFunction|SetDBNOSettings)))))

(event OnRemovedfromMinigame(FortMinigameProgress)
  (params (Minigame :type Fort%20Minigame%20Object%20Reference))
  ;[1 1 0] Added / Removed from minigame
  (Networking|SwitchHasAuthority
    (:Authority
      (Minigame|UnbindEventfromOnMinigamePlayerSpawned Minigame (EventDispatchers|CreateEvent :!event OnParticipantSpawned))
      (Minigame|UnbindEventfromPlayerAddedPostResetDelegate Minigame (EventDispatchers|CreateEvent :!event OnPlayerAddedToMinigame))
      (Minigame|UnbindEventfromPlayerRemovedDelegate Minigame (EventDispatchers|CreateEvent :!event Player%20Removed%20from%20Minigame))
      (Minigame|UnbindEventfromOnFinishedAddingMinigamePlayerDelegate Minigame (EventDispatchers|CreateEvent :!event OnFinishedAddingMinigamePlayer))
      (CallFunction|DisableDevice Minigame))))

(event Custom|DownPlayerReceivedFunction
  (params (Instigator :type Controller%20Object%20Reference))
  (bind _controller (CreativeAI|Controller|GetParticipantControllerfromActor Instigator))
  ;[0.1761 0 1 0.25] Receivers
  (Networking|SwitchHasAuthority
    (:Authority
      (if (Class|CreativeEnabledComponent|IsEnabled (Variables|Default|GetEnabledComponent))
        (Utilities|IsValid _controller
          (:Is%20Valid
            (if (Default|IsDBNOEnabled (Variables|/Script/Engine.Controller|GetPlayerState _controller))
              (bind _asfort_player_pawn_athena (Utilities|Casting|CastToFortPlayerPawnAthena (Pawn|GetControlledPawn Instigator)))
              (if (not (Health|IsDBNO _asfort_player_pawn_athena))
                (GameplayEffects|ApplyGameplayEffectToSelf :self (Ability|GetAbilitySystemComponent Instigator) :GameplayEffectClass (Variables|GameplayEffects|GetGE_KillPlayer) :Level 0.0)))))))))

(event Custom|RevivePlayerReceivedFunction
  (params (Instigator :type Controller%20Object%20Reference))
  ;[0.1761 0 1 0.25] Receivers
  (Networking|SwitchHasAuthority
    (:Authority
      (if (Class|CreativeEnabledComponent|IsEnabled (Variables|Default|GetEnabledComponent))
        (bind _asfort_player_pawn_athena (Utilities|Casting|CastToFortPlayerPawnAthena (Pawn|GetControlledPawn Instigator)))
        (if (Health|IsDBNO _asfort_player_pawn_athena)
          (Utilities|IsValid (Ability|GetAbilitySystemComponent (Class|FortPlayerPawn|GetDBNOHoister _asfort_player_pawn_athena))
            (:Is%20Valid
              (Abilities|TryActivateAbilitybyClass (Ability|GetAbilitySystemComponent (Class|FortPlayerPawn|GetDBNOHoister _asfort_player_pawn_athena)) "/Game/Abilities/NPC/Generic/GAB_DropPlayer.GAB_DropPlayer_C" true)
              (Utilities|FlowControl|DelayUntilNextTick)
              (Class|FortPlayerPawnAthena|ForceRevivefromDBNO _asfort_player_pawn_athena Instigator)
              (Transmitters|PawnRevivedfromDBNO _asfort_player_pawn_athena))
            (:Is%20Not%20Valid
              (Class|FortPlayerPawnAthena|ForceRevivefromDBNO _asfort_player_pawn_athena Instigator)
              (Transmitters|PawnRevivedfromDBNO _asfort_player_pawn_athena))))))))

(event Custom|EnabledReceivedFunction
  (params (Instigator :type Controller%20Object%20Reference))
  ;[0.1761 0 1 0.25] Receivers
  (Networking|SwitchHasAuthority
    (:Authority
      (Class|CreativeEnabledComponent|SetEnabled (Variables|Default|GetEnabledComponent) true))))

(event Custom|DisableReceivedFunction
  (params (Instigator :type Controller%20Object%20Reference))
  ;[0.1761 0 1 0.25] Receivers
  (Networking|SwitchHasAuthority
    (:Authority
      (Class|CreativeEnabledComponent|SetEnabled (Variables|Default|GetEnabledComponent) false))))

(event HandleVersionUpgrades(CreativeVersion)
  (params (LastSavedVersion :type Creative%20Version%20Id%20Structure))
  (CallFunction|VersionFixup LastSavedVersion))

(event MinigameStartedDelegate(FortMinigameProgress)
  (for _array_element (Minigame|GetParticipatingPlayers (Variables|/Script/FortniteGame.FortMinigameLogicComponent|GetCurrentMinigame (Variables|Default|GetFortMinigameProgress)))
    (CallFunction|EquipAbilitySet (PlayerState|GetPawn _array_element))))

(fn ApplyCreativeDBNOtoPlayer
  (params (Player%20State :type Player%20State%20Object%20Reference))
  (GameplayEffects|ApplyGameplayEffectToSelf :self (Ability|GetAbilitySystemComponent Player%20State) :GameplayEffectClass (Variables|GameplayEffects|GetGE_ApplyCreativeDBNO) :Level 0.0))

(fn ApplyDBNOCreative
  (for _array_element (Minigame|GetParticipatingPlayers (Variables|/Script/FortniteGame.FortMinigameLogicComponent|GetCurrentMinigame (Variables|Default|GetFortMinigameProgress)))
    (Transmitters|BindtoPawnDBNOStatesChange (Variables|/Script/Engine.PlayerState|GetPawnPrivate _array_element))
    (CallFunction|ApplyCreativeDBNOToPlayer (Variables|/Script/Engine.Pawn|GetPlayerState (Variables|/Script/Engine.PlayerState|GetPawnPrivate _array_element)))))

(fn BindToPawnDBNOStatesChange
  (params (Pawn :type Pawn%20Object%20Reference))
  (bind _asplayer_pawn_athena (Utilities|Casting|CastToPlayerPawn_Athena Pawn)
    (:then
      (EventDispatchers|BindEventtoOnDBNOStateChanged _asplayer_pawn_athena (EventDispatchers|CreateEvent :!event PawnDBNOStateChanged))
      (EventDispatchers|BindEventtoOnRevivedfromDBNODelegate _asplayer_pawn_athena (EventDispatchers|CreateEvent :!event PawnRevivedFromDBNO))
      (EventDispatchers|BindEventtoOnDBNOHoisterChangedDelegate _asplayer_pawn_athena (EventDispatchers|CreateEvent :!event PawnDBNOHoisterChanged)))
    (:CastFailed
      _))
  (return))

(fn BuildDataRegistryResolverScope
  (bind _self self)
  (bind _returnvalue (Minigame|GetMinigame _self :!class /Script/FortniteGame.MinigameBlueprintLibrary))
  ;[1 0.8475 0.2776] Gives the name of the volume as a scope to the data registry. Needed for calls to Get/SetValueWithContext to work as expected.
  (Utilities|IsValid _returnvalue
    (:Is%20Valid
      (return (Utilities|Array|MakeArray (SpatialGameplay|GetSpatialGameplayInterfaceFNameforActor _returnvalue)) 0 true))
    (:Is%20Not%20Valid
      (return 0 0 false))))

(fn DisableDevice
  (params (Minigame :type Fort%20Minigame%20Object%20Reference))
  (Class|FortAthenaMutatorDBNOSetting|UnsetDBNOSettings (Class|FortAthenaMutatorDBNOSetting|GetDBNOSettingsMutator Minigame) self)
  (Default|RemoveDBNOCreative Minigame))

(fn DropItemsFromPlayer
  (params (DropItemSetting :type Enum_DBNO_DropItemSetting%20Enum :default "NewEnumerator0") (PlayerPawn :type Fort%20Player%20Pawn%20Object%20Reference))
  (bind _emptycreativetags (LocalVariables|GetEmptyCreativeTags))
  (Utilities|IsValid PlayerPawn
    (:Is%20Valid
      (bind _asfort_player_controller (Utilities|Casting|CastToFortPlayerController (Pawn|GetController PlayerPawn)))
      (switch Utilities|FlowControl|Switch|SwitchonEnum_DBNO_DropItemSetting (Variables|GetDropItemSetting)
        (:NewEnumerator1
          (Inventory|DropAllItems _asfort_player_controller 0 0 false true :!class /Script/FortniteGame.FortPlayerController))
        (:NewEnumerator2
          (Inventory|DropItemsbyCreativeTags _asfort_player_controller (Variables|Default|GetBuildingMaterialItemDropTags) _emptycreativetags))
        (:NewEnumerator3
          (Inventory|DropItemsbyCreativeTags _asfort_player_controller (Variables|Default|GetGameResourceItemDropTags) _emptycreativetags))))))

(fn EquipAbilitySet
  (params (Pawn :type Pawn%20Object%20Reference))
  (bind _asfort_pawn (Utilities|Casting|CastToFortPawn Pawn))
  (Ability|EquipFortAbilitySet _asfort_pawn "/TrendConcentrate/AS_TrendConcentrate.AS_TrendConcentrate"))

(fn GetEnabledStateasIndex
  (return (DeviceConversion|EnabledPhaseEnumtoLegacyInt (Variables|UserOptions|GetEnabledDuringPhase))))

(fn GetCacheContextOverride
  (return 0))

(fn GetDecaySettings
  ;[0.6158 1 0.3112 0.5] Use these values to update players when they enter DBNO
  (return (Utilities|Enum|NotEqual(Enum) (Variables|UserOptions|GetReviveProgressDecay) "EInteractDecay_Pause") (CallFunction|GetReviveProgressDecayMulti) (Utilities|Enum|Equal(Enum) (Variables|UserOptions|GetReviveProgressDecay) "EInteractDecay_Reset")))

(fn GetReviveProgressDecayMulti
  (switch Utilities|FlowControl|Switch|SwitchonEInteractProgressDecay (Variables|UserOptions|GetReviveProgressDecay)
    (:EInteractDecay_Multi
      (return (Variables|UserOptions|GetDecayRateMultiplier)))
    (:EInteractDecay_Reset
      (return 0.0))
    (:EInteractDecay_Pause
      _)
    (:EInteractDecay_BR
      _))
  (return (ScalableFloat|GetValueatLevelwithContext (LocalVariables|GetEnableDecayRow_Default))))

(fn GetScaleMax
  (params (DefaultScaleMax :type Float%20%28single-precision%29))
  (return 1.0))

(fn GetScaleMin
  (params (DefaultScaleMin :type Float%20%28single-precision%29))
  (return 1.0))

(fn GetVisibilityComponentsWithCollisionOverrides
  (return (Utilities|Map|MakeMap (Variables|Debug|GetStaticMeshComponent) "NoCollision" 0 "NoCollision")))

(fn IsDBNOEnabled
  (params (PlayerState :type Player%20State%20Object%20Reference))
  (bind _asfort_game_state_athena (Utilities|Casting|CastToFortGameStateAthena (Game|GetGameState))
    (:then
      (bind _asfort_player_state (Utilities|Casting|CastToFortPlayerState PlayerState)
        (:then
          (return (and (Utilities|Enum|NotEqual(Enum) (Utilities|Struct|BreakDBNOCustomSettings (Class|FortGameStateAthena|EvaluateDBNOMutator _asfort_game_state_athena _asfort_player_state)) "Off") (.ReturnValue (Class|FortGameStateAthena|EvaluateDBNOMutator _asfort_game_state_athena _asfort_player_state)))))
        (:CastFailed
          _)))
    (:CastFailed
      _))
  (return false))

(fn MigrateLegacyUserOptions
  (bind _tenacitydepletionrate (Class|GetClassDefaults "/CRD_DBNO/Device_DBNO_V2.Device_DBNO_V2_C"))
  (bind (_reviveprogressdecay _decayratemultiplier _gradualresetsettings _predatesteamclassfix _ballowdodgeroll _enabledindex) (Class|GetClassDefaults (Utilities|GetClass self)))
  (bind _knob_affectsclass (Class|GetClassDefaults "/CRD_DBNO/Device_DBNO_V2.Device_DBNO_V2_C"))
  (bind _revivehealthpercentage (Class|GetClassDefaults "/CRD_DBNO/Device_DBNO_V2.Device_DBNO_V2_C"))
  (bind _knob_affectsteam (Class|GetClassDefaults "/CRD_DBNO/Device_DBNO_V2.Device_DBNO_V2_C"))
  (bind _enabledindex_2 (Variables|LegacyOptions|GetEnabledIndex))
  (|CallParentFunction)
  (Utilities|FlowControl|Sequence
    (:then_0
      ;[1 1 1] EnabledIndex -> EnabledDuringPhase
      ;;[1 1 1] V1 used custom options for Enabled Phases.
      (if (!= _enabledindex (Variables|LegacyOptions|GetEnabledIndex))
        (if (!= _enabledindex_2 -2)
          (Variables|UserOptions|SetEnabledDuringPhase (Utilities|Enum|BytetoEnumEEnabledOnPhase (Math|Conversions|ToByte(Integer) (Utilities|Select :Index _enabledindex_2 :Option%200 1 :Option%201 0 :Option%202 3 :Option%203 2))))
          (else
            (Variables|UserOptions|SetEnabledDuringPhase "NewEnumerator1")))
        (LocalVariables|SetOptionsMigrated true)))
    (:then_1
      ;[1 1 1] TenacityDepletionRate -> CustomTenacityDepletionRate , UseDefaultTenacityDepletionRate
      (if (!= _tenacitydepletionrate (Variables|Legacy|GetTenacityDepletionRate))
        (Variables|UserOptions|Basic|SetUseDefaultTenacityDepletionRate false)
        (Variables|UserOptions|Basic|SetCustomTenacityDepletionRate (Math|Float|Truncate (Variables|Legacy|GetTenacityDepletionRate)))
        (Variables|Legacy|SetTenacityDepletionRate _tenacitydepletionrate)
        (LocalVariables|SetOptionsMigrated true)))
    (:then_2
      ;[1 1 1] ReviveHealthPercentage -> UseDefaultHealthAfterRevive , HealthAfterRevive
      (if (!= _revivehealthpercentage (Variables|Legacy|GetReviveHealthPercentage))
        (Variables|UserOptions|SetUseDefaultHealthAfterRevive false)
        (Variables|UserOptions|SetHealthAfterRevive (* (Variables|Legacy|GetReviveHealthPercentage) 1e+02))
        (Variables|Legacy|SetReviveHealthPercentage _revivehealthpercentage)
        (LocalVariables|SetOptionsMigrated true)))
    (:then_3
      ;[1 1 1] Knob_AffectsTeam -> InvertTeamSelection
      (if (!= _knob_affectsteam (Variables|Legacy|GetKnob_AffectsTeam))
        (Variables|UserOptions|SetInvertTeamSelection false)
        (Variables|Legacy|SetKnob_AffectsTeam _knob_affectsteam)
        (LocalVariables|SetOptionsMigrated true)))
    (:then_4
      ;[1 1 1] Knob_AffectsClass -> InvertClassSelection
      (if (!= _knob_affectsclass (Variables|Legacy|GetKnob_AffectsClass))
        (Variables|UserOptions|SetInvertClassSelection false)
        (Variables|Legacy|SetKnob_AffectsClass _knob_affectsclass)
        (LocalVariables|SetOptionsMigrated true)))
    (:then_5
      (if (LocalVariables|GetOptionsMigrated)
        (OptionData|RefreshPropertyOverrides (Variables|Default|GetToyOptionsComponent))))))

(fn OnFinishedAddingMinigamePlayer
  (params (UniqueNetId :type Unique%20Net%20Id%20Structure) (bIsLocalPlayer :type Boolean) (FortPlayerState :type Fort%20Player%20State%20Object%20Reference))
  (CallFunction|EquipAbilitySet (PlayerState|GetPawn FortPlayerState)))

(fn OnParticipantSpawned
  (params (SpawningPlayer :type Controller%20Object%20Reference))
  (bind _returnvalue (Pawn|GetControlledPawn SpawningPlayer))
  (Transmitters|BindtoPawnDBNOStatesChange _returnvalue)
  (CallFunction|ApplyCreativeDBNOToPlayer (Variables|/Script/Engine.Pawn|GetPlayerState _returnvalue)))

(fn OnPlayerAddedToMinigame
  (params (PlayerNetid :type Unique%20Net%20Id%20Structure) (IsLocalPlayer :type Boolean))
  (bind _fort_player_state (Class|CreativeCommonDeviceFunctionLibrary|GetPlayerStateFromNetID PlayerNetid))
  (bind _pawnprivate (Variables|/Script/Engine.PlayerState|GetPawnPrivate _fort_player_state))
  (Transmitters|BindtoPawnDBNOStatesChange _pawnprivate)
  (CallFunction|ApplyCreativeDBNOToPlayer (Variables|/Script/Engine.Pawn|GetPlayerState _pawnprivate)))

(fn PawnDBNOHoisterChanged
  (params (CarryEvent :type EFortDBNOCarryEvent%20Enum :default "PickedUp") (HoisterPawn :type Fort%20Player%20Pawn%20Object%20Reference) (HoisteePawn :type Fort%20Player%20Pawn%20Object%20Reference))
  (bind (_value _returnvalue) (Utilities|Map|Find (Variables|Default|GetPlayersLastState) (Variables|/Script/Engine.Pawn|GetPlayerState (Variables|GetHoisteePawn))))
  (Utilities|IsValidforPawn HoisteePawn
    (:true
      (Utilities|FlowControl|Sequence
        (:then_0
          (switch Utilities|FlowControl|Switch|SwitchonEFortDBNOCarryEvent CarryEvent
            (:PickedUp
              (GameplayEvent|BroadcastEvent (Variables|UserOptions|Events|GetOn%20Player%20Picked%20Up) 0 (Utilities|Struct|MakeGameplayEventPlayerPickedUp (Pawn|GetController (Variables|GetHoisteePawn)))))
            (:Interrogating
              (GameplayEvent|BroadcastEvent (Variables|UserOptions|Events|GetOn%20Shaken%20Down) 0 (Utilities|Struct|MakeGameplayEventShakenDown (Pawn|GetController (Variables|GetHoisteePawn))))
              (GameplayEvent|BroadcastEvent (Variables|UserOptions|Events|GetOn%20Shakedown) 0 (Utilities|Struct|MakeGameplayEventShakedown (Pawn|GetController (Variables|GetHoisterPawn))))
              (CallFunction|DropItemsfromPlayer (Variables|Knobs|GetDropItemsOnShakenDown) (Variables|GetHoisterPawn)))
            (:Dropped
              (if (and (Utilities|Enum|Equal(Enum) _value "PickedUp") _returnvalue)
                (GameplayEvent|BroadcastEvent (Variables|UserOptions|Events|GetOn%20Player%20Dropped) 0 (Utilities|Struct|MakeGameplayEventPlayerDropped (Pawn|GetController (Variables|GetHoisteePawn))))))
            (:Thrown
              (GameplayEvent|BroadcastEvent (Variables|UserOptions|Events|GetOn%20Player%20Thrown) 0 (Utilities|Struct|MakeGameplayEventPlayerThrown (Pawn|GetController (Variables|GetHoisteePawn)))))))
        (:then_1
          (Utilities|Map|Add (Variables|Default|GetPlayersLastState) (Variables|/Script/Engine.Pawn|GetPlayerState (Variables|GetHoisteePawn)) (Variables|GetCarryEvent)))))))

(fn PawnDBNOStateChanged
  (params (FortPawn :type Fort%20Pawn%20Object%20Reference) (bInIsDBNO :type Boolean))
  (if bInIsDBNO
    (Utilities|IsValidforPawn FortPawn
      (:true
        (GameplayEvent|BroadcastEvent (Variables|UserOptions|Events|GetOn%20Player%20Downed) 0 (Utilities|Struct|MakeGameplayEventPlayerDowned (Pawn|GetController (Variables|GetFortPawn))))
        (bind _asfort_player_pawn (Utilities|Casting|CastToFortPlayerPawn (Variables|GetFortPawn)))
        (CallFunction|DropItemsfromPlayer (Variables|Knobs|GetDropItemsOnKnockDown) _asfort_player_pawn)))))

(fn PawnRevivedFromDBNO
  (params (RevivedPawn :type Fort%20Player%20Pawn%20Object%20Reference))
  (Utilities|IsValidforPawn RevivedPawn
    (:true
      (GameplayEvent|BroadcastEvent (Variables|UserOptions|Events|GetOn%20Player%20Revived) 0 (Utilities|Struct|MakeGameplayEventPlayerRevived (Pawn|GetController RevivedPawn))))))

(fn PlayerRemovedfromMinigame
  (params (UniqueNetId :type Unique%20Net%20Id%20Structure) (bIsLocalPlayer :type Boolean))
  (Transmitters|UnBindtoPawnDBNOStatesChange (Variables|/Script/Engine.PlayerState|GetPawnPrivate (Class|CreativeCommonDeviceFunctionLibrary|GetPlayerStateFromNetID UniqueNetId))))

(fn RemoveDBNOCreative
  (params (Minigame :type Fort%20Minigame%20Object%20Reference))
  (for _array_element (Minigame|GetParticipatingPlayers Minigame)
    (Transmitters|UnBindtoPawnDBNOStatesChange (Variables|/Script/Engine.PlayerState|GetPawnPrivate _array_element))))

(fn SetDBNOSettings
  (bind (_ballowdecayandreset _reviveprogressdecaymulti _bendclientlerpwhenreviveinterrupted) (Default|GetDecaySettings))
  (Class|FortAthenaMutatorDBNOSetting|SetDBNOSettingsandOverride (Class|FortAthenaMutatorDBNOSetting|GetDBNOSettingsMutator (Variables|/Script/FortniteGame.FortMinigameLogicComponent|GetCurrentMinigame (Variables|Default|GetFortMinigameProgress))) (Utilities|Struct|MakeDBNOCustomSettings (Variables|UserOptions|Basic|GetDBNOEnabled) (Math|Conversions|ToFloat(Integer) (Math|Integer|SelectInt -1 (Variables|UserOptions|Basic|GetCustomTenacityDepletionRate) (Variables|UserOptions|Basic|GetUseDefaultTenacityDepletionRate))) (Math|Float|SelectFloat 0.0 (/ (Variables|UserOptions|GetHealthAfterRevive) 1e+02) (Variables|UserOptions|GetUseDefaultHealthAfterRevive)) (Variables|UserOptions|GetbAllowRevives) (Variables|UserOptions|GetbAllowCarry) (Math|Float|SelectFloat 0.01 (Variables|UserOptions|GetTimeToRevive) (== (Variables|UserOptions|GetTimeToRevive) 0)) (Variables|UserOptions|GetbAlertTeamWhenDowned) (Variables|UserOptions|GetbLastManStandingMode) (Variables|UserOptions|GetShakedowns) (Variables|UserOptions|GetShowPlayerLocationOnShakedown) (Creative|TeamOption|GetLegacyTeamIndex (Variables|UserOptions|GetSelectedTeam)) (Variables|UserOptions|GetInvertTeamSelection) (Creative|ClassOption|GetLegacyClassIndex (Variables|UserOptions|GetSelected%20Class)) (Variables|UserOptions|GetInvertClassSelection) (Variables|UserOptions|GetTenacityType) (Math|Conversions|ToFloat(Integer) (Variables|UserOptions|GetTenacityAmount)) (Variables|UserOptions|GetDownedInvincibilityTime_Override) (Variables|UserOptions|GetDownedInvincibilityTime) _ballowdecayandreset _reviveprogressdecaymulti _bendclientlerpwhenreviveinterrupted (Variables|UserOptions|GetbAllowDodgeRoll) (Variables|UserOptions|GetbAllowSprinting) (Variables|UserOptions|GetSprintingStaminaCost) (Variables|UserOptions|GetbAllowRagdoll) (Variables|UserOptions|GetbAllowHurdle) (Variables|UserOptions|GetAllowInteractions)) self))

(fn UnBindToPawnDBNOStatesChange
  (params (Pawn :type Pawn%20Object%20Reference))
  (bind _asplayer_pawn_athena (Utilities|Casting|CastToPlayerPawn_Athena Pawn)
    (:then
      (EventDispatchers|UnbindEventfromOnDBNOStateChanged _asplayer_pawn_athena (EventDispatchers|CreateEvent :!event PawnDBNOStateChanged))
      (EventDispatchers|UnbindEventfromOnDBNOHoisterChangedDelegate _asplayer_pawn_athena (EventDispatchers|CreateEvent :!event PawnDBNOHoisterChanged))
      (EventDispatchers|UnbindEventfromOnRevivedfromDBNODelegate _asplayer_pawn_athena (EventDispatchers|CreateEvent :!event PawnRevivedFromDBNO)))
    (:CastFailed
      _))
  (return))

(fn UpdateVisibilityInGame)

(fn ConstructionScript
  (|CallParentFunction))

(fn VersionFixup
  (params (Last%20Version :type Creative%20Version%20Id%20Structure))
  (bind _selected_class_classtype (Utilities|Struct|BreakCreativeClassOption (Variables|GetSelected%20Class_ClassType)))
  ;[1 0 0.6455 0.2] Fixing an issue with older devices where Class:Any was accidentally meaning Class: Not None!. Keeping this functionality for older devices but ignoring it for newer ones.
  (Utilities|FlowControl|Sequence
    (:then_0
      (CreativeVersion|CompareVersiontoTag Last%20Version "(TagName=\"Creative.Version.Release.31.10\")"
        (:OlderVersion
          (Variables|Legacy|SetPredatesTeamClassFix true) ; DBNO wasn't handling Team or Class correctly\nat all for tracking info. Fixing, but leaving old logic in place\nto avoid breaking old experiences.
          (if (and (Utilities|Enum|Equal(Enum) _selected_class_classtype "Any") (not (Variables|UserOptions|GetInvertClassSelection)))
            (Variables|SetSelected%20Class_ClassType (struct :ClassType "NoClass" :ClassSlot 1))
            (Variables|UserOptions|SetInvertClassSelection true)
            (LocalVariables|SetOptions%20Refresh%20Required true)))))
    (:then_1
      (CreativeVersion|CompareVersiontoTag Last%20Version "(TagName=\"Creative.Version.Release.31.30\")"
        (:OlderVersion
          (Utilities|FlowControl|Sequence
            (:then_0
              (if (Creative|TeamOption|IsTeamType (Variables|UserOptions|GetSelectedTeam) "None")
                (Variables|SetSelectedTeam_TeamType (struct :TeamType "Any" :TeamIndex 1))
                (Variables|UserOptions|SetInvertTeamSelection false)))
            (:then_1
              (if (Creative|ClassOption|IsClassType (Variables|UserOptions|GetSelected%20Class) "All")
                (Variables|SetSelected%20Class_ClassType (struct :ClassType "Any" :ClassSlot 1)))))
          (LocalVariables|SetOptions%20Refresh%20Required true))))
    (:then_2
      (CreativeVersion|CompareVersiontoTag Last%20Version "(TagName=\"Creative.Version.Release.39.00\")"
        (:OlderVersion
          ;[1 1 1] New options added as part of FORT-958824 default to off for migrated devices, but on for newly placed devices
          (Variables|UserOptions|SetbAllowDodgeRoll false)
          (Variables|UserOptions|SetbAllowSprinting false)
          (Variables|UserOptions|SetbAllowRagdoll false)
          (Variables|UserOptions|SetbAllowHurdle false)
          (Variables|UserOptions|SetAllowInteractions "Improved")
          (LocalVariables|SetOptions%20Refresh%20Required true))))
    (:then_3
      (if (LocalVariables|GetOptions%20Refresh%20Required)
        (OptionData|RefreshPropertyOverrides (Variables|Default|GetToyOptionsComponent))))))
